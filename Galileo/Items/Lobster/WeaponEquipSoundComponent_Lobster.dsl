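; Equip sound component for the Lobster weapon, driving its equip, loop and unequip audio through sound library gameplay-tag events.
; Loop updates play a quick or normal equip event plus the loop, or stop whichever are active.
; Instant FX sets the quick-equip flag on cosmetic-capable instances before playing audio.
; Unequipping and vehicle changes can fire the loop-stop event, with vehicle changes also updating in-vehicle and holster state.
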(event EventOnUnEquipped
  (|CallParentFunction)
  (CallFunction|PlayUnEquipFX false))

(event EventOnVehicleStateChanged
  (params (FortPlayerPawn :type Fort%20Player%20Pawn%20Object%20Reference) (NewVehicle :type Actor%20Object%20Reference) (OldVehicle :type Actor%20Object%20Reference))
  (bind _output_get (Variables|Private|SetbIsInVehicle (and (Vehicle|IsinVehicle (Variables|Private|GetPlayerPawn)) (not (CallFunction|OverrideVehicleFuncions)))))
  (if (and (Utilities|IsValid NewVehicle :!class /Script/Engine.KismetSystemLibrary) (!= NewVehicle OldVehicle))
    (CallFunction|PlayUnEquipFX true)
    (else
      _))
  (if _output_get
    _
    (elif (Variables|Private|GetbIsVisible)
      _
      (else
        (CallFunction|OnHolestered))))
  (CallFunction|OnUnHolestered))

(fn PlayValidEvent
  (params (GameplayTag :type Gameplay%20Tag%20Structure))
  (Utilities|IsValid (Variables|Private|GetSoundLibraryComponent)
    (:Is%20Valid
      (bind _result (|CallParentFunction GameplayTag))
      (return _result))))

(fn PlayAudioInstantFX
  (Networking|CanExecuteCosmeticEvents
    (:True
      (if (not (Variables|Private|GetAudioPlaying))
        (Variables|AudioQuickPlay|SetAudioQuickOpen true)
        (CallFunction|PlayAudio)))))

(fn PlayUnEquipFX
  (params (AlwaysPlayIfValid :type Boolean))
  (if AlwaysPlayIfValid
    _
    (elif (Variables|Private|GetbIsInVehicle)
      _))
  (CallFunction|PlayValidEvent (Variables|Event|GetLoopingStopEventName)))

(fn UpdateLoopStatus
  (params (Play :type Boolean))
  (bind _soundlibrarycomponent (Variables|Private|GetSoundLibraryComponent))
  (bind _quickequipeventname (Variables|Event|GetQuickEquipEventName))
  (bind _equipeventname (Variables|Event|GetEquipEventName))
  (bind _soundlibrarycomponent_2 (Variables|Private|GetSoundLibraryComponent))
  (bind _loopeventname (Variables|Event|GetLoopEventName))
  (bind _soundlibrarycomponent_3 (Variables|Private|GetSoundLibraryComponent))
  (if Play
    (if (Variables|AudioQuickPlay|GetAudioQuickOpen)
      (CallFunction|PlayValidEvent (Variables|Event|GetQuickEquipEventName))
      (Variables|AudioQuickPlay|SetAudioQuickOpen false)
      (else
        (CallFunction|PlayValidEvent (Variables|Event|GetEquipEventName))))
    (CallFunction|PlayValidEvent (Variables|Event|GetLoopEventName))
    (else
      (if (SoundLibrary|IsEventActive _soundlibrarycomponent_3 _quickequipeventname)
        (SoundLibrary|StopEvent _soundlibrarycomponent_3 _quickequipeventname)
        (else
          _))
      (if (SoundLibrary|IsEventActive _soundlibrarycomponent _equipeventname)
        (SoundLibrary|StopEvent _soundlibrarycomponent _equipeventname)
        (else
          _))
      (if (SoundLibrary|IsEventActive _soundlibrarycomponent_2 _loopeventname)
        (SoundLibrary|StopEvent _soundlibrarycomponent_2 _loopeventname)))))
